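# Sound cue for the looping ambient audio of a memory card pickaxe collectible, a BattlePassS20 quest actor.
# A looping wave player feeds an enveloper whose volume curve fades the sound in, then a modulator at the root.
# The cue routes to the quest sound class, uses its own attenuation asset for distance falloff, and overrides concurrency to cap simultaneous instances.
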SoundCue /BattlePassS20/QuestActors/PickaxeCollectible/Audio/MemoryCard_Ambient_Loop_Cue
root=Modulator_0
nodes [
 {
  id=Enveloper_0
  children [WavePlayer_0]
  properties {pitchCurve {editorCurveData {keys [{time=0}]}}, volumeCurve {editorCurveData {keys [<at 0 0>,<at 1 1>]}}}
 }
 {id=Modulator_0, children [Enveloper_0]}
 {id=WavePlayer_0, properties {bLooping=true, soundWaveAssetPtr=&/MemoryCard_AmbientLoop}}
]
properties {
 attenuationSettings=&/MemoryCard_Ambient_Loop_Att
 bOverrideConcurrency=true
 maxDistance=1500
 soundClassObject=&Fort_Quests
 volumeMultiplier=1.5
 attenuationOverrides {falloffDistance=700, attenuationShapeExtents {x=100}}
 concurrencyOverrides {maxCount=3}
}
