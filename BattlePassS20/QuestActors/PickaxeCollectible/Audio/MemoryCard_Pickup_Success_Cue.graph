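# Sound cue played on a successful memory card pickup, part of the Season 20 Battle Pass pickaxe collectible quest actors.
# A quality level node sends its first two levels to an equally weighted random pick of three collect waves, and its last two to the first wave alone.
# It reuses the memory card ambient loop's attenuation and plays at reduced volume in the non-critical in-game UI sound class.
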SoundCue /BattlePassS20/QuestActors/PickaxeCollectible/Audio/MemoryCard_Pickup_Success_Cue
root=QualityLevel_0
nodes [
 {id=QualityLevel_0, children [Random_0,Random_0,WavePlayer_4,WavePlayer_4]}
 {id=Random_0, children [WavePlayer_5,WavePlayer_6,WavePlayer_7], properties {bShouldExcludeFromBranchCulling=true, weights [1,1,1]}}
 {id=WavePlayer_4, properties {soundWaveAssetPtr=&/CollectMemoryCard_01}}
 {id=WavePlayer_5, properties {soundWaveAssetPtr=&CollectMemoryCard_01}}
 {id=WavePlayer_6, properties {soundWaveAssetPtr=&/CollectMemoryCard_02}}
 {id=WavePlayer_7, properties {soundWaveAssetPtr=&/CollectMemoryCard_03}}
]
properties {
 attenuationSettings=&/MemoryCard_Ambient_Loop_Att
 maxDistance=1500
 soundClassObject=&Fort_UI_InGame_NonCritical
 volumeMultiplier=0.6
}
