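; Parent Blueprint for a Season 20 Battle Pass quest pickaxe collectible, implementing only color setup for its effects.
; The construction script calls the parent, then reads a named color from a material parameter collection.
; It feeds that color, through a multiply, to the ambient Niagara light shaft and to several color parameters on the post-interaction despawn particle system.
; BeginPlay, overlap and Tick are disabled stubs, so this graph adds no runtime logic.
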(event EventBeginPlay
  (|CallParentFunction) ; This node is disabled and will not be called.\nDrag off pins to build functionality.
)

(event Collision|EventActorBeginOverlap
  (params (OtherActor :type Actor%20Object%20Reference))
  (|CallParentFunction OtherActor) ; This node is disabled and will not be called.\nDrag off pins to build functionality.
)

(event EventTick
  (params (DeltaSeconds :type Float%20%28single-precision%29))
  (|CallParentFunction DeltaSeconds) ; This node is disabled and will not be called.\nDrag off pins to build functionality.
)

(fn ConstructionScript
  (|CallParentFunction)
  ;[1 1 1] Get color from color collection.
  (bind _returnvalue (Rendering|Material|GetVectorParameterValue (Variables|Default|GetColorParameterCollection) (Variables|Default|GetColorName) :!class /Script/Engine.KismetMaterialLibrary))
  ;[0.7496 1 0.725] Set Niagara Pre-Interact Ambient Color (LIGHT SHAFT)
  (Niagara|SetNiagaraVariable(LinearColor) (Variables|Components|GetN_Ambient) "GenericPropColor" (Math|Color|LinearColor*LinearColor _returnvalue "(R=1.000000,G=1.000000,B=1.000000,A=5.000000)"))
  ;[0.7496 1 0.725] Set Post-Interaction Ambient Particle (despawn) Component colors.
  (Effects|Components|ParticleSystem|SetVectorParameter (Variables|Components|GetP_PostInteractionAmbient) "SweepColor" (* (Math|Conversions|ToVector(LinearColor) _returnvalue) 1.0))
  (Effects|Components|ParticleSystem|SetVectorParameter (Variables|Components|GetP_PostInteractionAmbient) "WideSparkColor" (* (Math|Conversions|ToVector(LinearColor) _returnvalue) 1.0))
  (Effects|Components|ParticleSystem|SetVectorParameter (Variables|Components|GetP_PostInteractionAmbient) "FlashColor" (* (Math|Conversions|ToVector(LinearColor) _returnvalue) 1.0))
  (Effects|Components|ParticleSystem|SetVectorParameter (Variables|Components|GetP_PostInteractionAmbient) "CoronaColor" (* (Math|Conversions|ToVector(LinearColor) _returnvalue) 1.0))
  (Effects|Components|ParticleSystem|SetVectorParameter (Variables|Components|GetP_PostInteractionAmbient) "ResidualSparksColor" (* (Math|Conversions|ToVector(LinearColor) _returnvalue) 1.0)))
